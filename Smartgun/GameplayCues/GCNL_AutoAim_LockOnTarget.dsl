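; Looping GameplayCue notify that drives the Smartgun auto-aim lock-on particle effect on a targeted player or wildlife pawn.
; Loop start caches target, instigator and effect component, shifting it on wildlife by capsule center of mass minus FX_Chest socket location.
; Tick feeds lock-on progress and count from the instigator's AutoAimWeaponPawnComponent into particle parameters.
; The local controller's vehicle seat-change delegates deactivate the effect, then reactivate it on completion.
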(event EventBeginPlay
  (bind _asfort_player_controller (Utilities|Casting|CastToFortPlayerController (Game|GetPlayerController 0)))
  (Vehicle|AssignOnVehicleSeatChanged _asfort_player_controller (delegate Custom|OnVehicleSeatChanged_Event))
  (Vehicle|AssignOnVehicleSeatChangedComplete _asfort_player_controller (delegate Custom|OnVehicleSeatChangedComplete_Event)))

(event Custom|OnVehicleSeatChangedComplete_Event
  (Components|Activation|SetActive (Variables|Default|GetNS%20Looping) true true))

(event Custom|OnVehicleSeatChanged_Event
  (params (SeatIndex :type Integer))
  (Components|Activation|SetActive (Variables|Default|GetNS%20Looping) false true))

(event Collision|EventActorBeginOverlap
  (params (OtherActor :type Actor%20Object%20Reference)))

(event EventTick
  (params (DeltaSeconds :type Float%20%28single-precision%29))
  (bind _ns_looping (Variables|Default|GetNS%20Looping))
  (Utilities|IsValid (Variables|Default|GetTarget%20Pawn)
    (:Is%20Valid
      (bind _asfort_pawn (Utilities|Casting|CastToFortPawn (Variables|Default|GetInstigator_0)))
      (bind _returnvalue (AutoAimWeapon|GetAutoAimWeaponPawnComponent _asfort_pawn))
      (Effects|Components|ParticleSystem|SetFloatParameter _ns_looping "LockProgress" (Class|AutoAimWeaponPawnComponent|GetLockonProgress _returnvalue))
      (Effects|Components|ParticleSystem|SetIntParameter _ns_looping "BulletCount" (Class|AutoAimWeaponPawnComponent|GetCurrentLockonCount _returnvalue :!class /Script/AutoAimWeaponRuntime.AutoAimWeaponPawnComponent)))))

(event EventOnLoopingStart(EitherVFX-preferthisversion)
  (params (MyTarget :type Actor%20Object%20Reference) (Parameters :type Gameplay%20Cue%20Parameters%20Structure%20%28by%20ref%29) (ParticleComponents :type Array%20of%20FXSystem%20Component%20Object%20References) (AudioComponents :type Array%20of%20Audio%20Component%20Object%20References))
  (bind (_normalizedmagnitude _rawmagnitude _effectcontext _matchedtagname _originaltag _aggregatedsourcetags _aggregatedtargettags _location _normal _instigator) (Ability|GameplayCue|BreakGameplayCueParameters Parameters))
  (bind _output (Utilities|Array|Get(acopy) ParticleComponents))
  (bind _asplayer_pawn_athena (Utilities|Casting|CastToPlayerPawn_Athena MyTarget)
    (:then
      (Variables|Default|SetTarget%20Pawn _asplayer_pawn_athena)
      (Variables|Default|SetInstigator_0 _instigator)
      (Variables|Default|SetNS%20Looping _output))
    (:CastFailed
      (bind _asnpc_pawn_wildlife_parent (Utilities|Casting|CastToNPC_Pawn_Wildlife_Parent MyTarget))
      (Variables|Default|SetTarget%20Pawn _asnpc_pawn_wildlife_parent)
      (Variables|Default|SetInstigator_0 _instigator)
      (bind _output_get (Variables|Default|SetNS%20Looping _output))
      (Transformation|SetRelativeLocation _output_get (- (Physics|GetCenterOfMass (Variables|/Script/Engine.Character|GetCapsuleComponent _asnpc_pawn_wildlife_parent) "None") (Transformation|GetSocketLocation (Variables|/Script/Engine.Character|GetMesh _asnpc_pawn_wildlife_parent) "FX_Chest")) false false))))

(fn ConstructionScript)
