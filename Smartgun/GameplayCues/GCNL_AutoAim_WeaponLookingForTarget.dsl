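; Looping gameplay cue that drives the Smartgun auto-aim effect shown while the weapon looks for a target.
; On loop start it caches the weapon actor, its instigator and the looping particle component, setting Weapon_Index to 1 for the auto-aim rifle when RawMagnitude equals 1.
; Each tick it reads the instigator pawn's AutoAimWeaponPawnComponent and feeds LockProgress, HasTarget and, when a target exists, TargetPositionV into the effect.
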(event EventBeginPlay)

(event Collision|EventActorBeginOverlap
  (params (OtherActor :type Actor%20Object%20Reference)))

(event EventTick
  (params (DeltaSeconds :type Float%20%28single-precision%29))
  (bind _ns_looping (Variables|Default|GetNS%20Looping))
  (Utilities|IsValid (Variables|Default|GetWeaponActor)
    (:Is%20Valid
      (bind _asfort_pawn (Utilities|Casting|CastToFortPawn (Variables|Default|GetInstigator_0)))
      (bind _returnvalue (AutoAimWeapon|GetAutoAimWeaponPawnComponent _asfort_pawn))
      (Utilities|IsValid _returnvalue
        (:Is%20Valid
          (Effects|Components|ParticleSystem|SetFloatParameter _ns_looping "LockProgress" (Class|AutoAimWeaponPawnComponent|GetLockonProgress _returnvalue))
          (bind (_outlocktargetlocation _returnvalue_2) (Class|AutoAimWeaponPawnComponent|GetLockonTargetLocation _returnvalue))
          (Effects|Components|ParticleSystem|SetBoolParameter _ns_looping "HasTarget" _returnvalue_2)
          (if _returnvalue_2
            (Effects|Components|ParticleSystem|SetVectorParameter _ns_looping "TargetPositionV" _outlocktargetlocation)))))))

(event EventOnLoopingStart(EitherVFX-preferthisversion)
  (params (MyTarget :type Actor%20Object%20Reference) (Parameters :type Gameplay%20Cue%20Parameters%20Structure%20%28by%20ref%29) (ParticleComponents :type Array%20of%20FXSystem%20Component%20Object%20References) (AudioComponents :type Array%20of%20Audio%20Component%20Object%20References))
  (bind (_normalizedmagnitude _rawmagnitude) (Ability|GameplayCue|BreakGameplayCueParameters Parameters))
  (Variables|Default|SetWeaponActor MyTarget)
  (Variables|Default|SetNS%20Looping (Utilities|Array|Get(acopy) ParticleComponents))
  (Variables|Default|SetInstigator_0 (Variables|/Script/Engine.Actor|GetInstigator (Variables|Default|GetWeaponActor)))
  ;[1 0.2276 0 0.5] AUTOAIM_RIFLE LOGIC TO FEED INDEX FOR AIMING NS
  (if (== _rawmagnitude 1.0)
    (Effects|Components|ParticleSystem|SetIntParameter (Variables|Default|GetNS%20Looping) "Weapon_Index" 1)))

(fn ConstructionScript)
